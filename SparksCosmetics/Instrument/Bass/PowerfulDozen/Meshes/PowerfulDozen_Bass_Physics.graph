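# Physics asset for the PowerfulDozen bass instrument cosmetic, giving its decorative ribbon jointed physics bodies.
# Box bodies cover the root and handle, while small identical capsules form a front chain of six segments and a back chain of four, both branching from Ribbon_Main.
# Ribbon joints use tight twist and moderate swing limits that mostly widen along each chain before narrowing at the tip.
# The root-to-handle and handle-to-Ribbon_Main joints leave all angular motion free.
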PhysicsAsset /SparksCosmetics/Instrument/Bass/PowerfulDozen/Meshes/PowerfulDozen_Bass_Physics
bodies {
 Ribbon_Back_01 [{shapeType=Capsule, length=1, radius=1, center=<0 0 0.3325>}]
 Ribbon_Back_02 [{shapeType=Capsule, length=1, radius=1, center=<0 0 0.3325>}]
 Ribbon_Back_03 [{shapeType=Capsule, length=1, radius=1, center=<0 0 0.3325>}]
 Ribbon_Back_04 [{shapeType=Capsule, length=1, radius=1, center=<0 0 0.3325>}]
 Ribbon_Front_01 [{shapeType=Capsule, length=1, radius=1, center=<0 0 0.3325>}]
 Ribbon_Front_02 [{shapeType=Capsule, length=1, radius=1, center=<0 0 0.3325>}]
 Ribbon_Front_03 [{shapeType=Capsule, length=1, radius=1, center=<0 0 0.3325>}]
 Ribbon_Front_04 [{shapeType=Capsule, length=1, radius=1, center=<0 0 0.3325>}]
 Ribbon_Front_05 [{shapeType=Capsule, length=1, radius=1, center=<0 0 0.3325>}]
 Ribbon_Front_06 [{shapeType=Capsule, length=1, radius=1, center=<0 0 0.3325>}]
 Ribbon_Main [{shapeType=Capsule, length=1, radius=1, center=<0 0 0.3325>}]
 handle [{shapeType=Box, extentX=4.763, extentY=2.2549, extentZ=20.0263, shapeName=handle_box, center=<-0.038 0.0992 65.1044>, rotation=<rot 0 -90 0>}]
 root [{shapeType=Box, extentX=16.0834, extentY=40.8228, extentZ=127.4115, shapeName=root_box, center=<0 0 37.8976>, rotation=<rot 0 -90 0>}]
}
constraints [
 {bone1Name=Ribbon_Back_01, bone2Name=Ribbon_Main, swing1LimitDegrees=25, swing2LimitDegrees=25, twistLimitDegrees=5}
 {bone1Name=Ribbon_Back_02, bone2Name=Ribbon_Back_01, swing1LimitDegrees=30, swing2LimitDegrees=35, twistLimitDegrees=10}
 {bone1Name=Ribbon_Back_03, bone2Name=Ribbon_Back_02, swing1LimitDegrees=35, swing2LimitDegrees=40, twistLimitDegrees=10}
 {bone1Name=Ribbon_Back_04, bone2Name=Ribbon_Back_03, swing1LimitDegrees=30, swing2LimitDegrees=35, twistLimitDegrees=10}
 {bone1Name=Ribbon_Front_01, bone2Name=Ribbon_Main, swing1LimitDegrees=20, swing2LimitDegrees=25, twistLimitDegrees=5}
 {bone1Name=Ribbon_Front_02, bone2Name=Ribbon_Front_01, swing1LimitDegrees=20, swing2LimitDegrees=35, twistLimitDegrees=10}
 {bone1Name=Ribbon_Front_03, bone2Name=Ribbon_Front_02, swing1LimitDegrees=30, swing2LimitDegrees=40, twistLimitDegrees=10}
 {bone1Name=Ribbon_Front_04, bone2Name=Ribbon_Front_03, swing1LimitDegrees=35, swing2LimitDegrees=45, twistLimitDegrees=10}
 {bone1Name=Ribbon_Front_05, bone2Name=Ribbon_Front_04, swing1LimitDegrees=40, swing2LimitDegrees=50, twistLimitDegrees=10}
 {bone1Name=Ribbon_Front_06, bone2Name=Ribbon_Front_05, swing1LimitDegrees=30, swing2LimitDegrees=40, twistLimitDegrees=10}
 {
  bone1Name=Ribbon_Main
  bone2Name=handle
  swing1LimitDegrees=45
  swing1Motion=Free
  swing2LimitDegrees=45
  swing2Motion=Free
  twistLimitDegrees=45
  twistMotion=Free
 }
 {
  bone1Name=handle
  bone2Name=root
  swing1LimitDegrees=45
  swing1Motion=Free
  swing2LimitDegrees=45
  swing2Motion=Free
  twistLimitDegrees=45
  twistMotion=Free
 }
]
